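# Looping interior water-on-rocks spot ambience cue for the Asteria Sky Island POI, in the ambient water sound class.
# A Mixer layers three reused loops: cave river, interior waterfall bottom and ATR waterfall.
# Each loop has its own cone attenuation override with a different falloff, the waterfall bottom reaching farthest.
# The shortest-range ATR layer adds interpolated listener focus, and all quality levels share the same mix.
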SoundCue /Game/Sounds/Asteria/POI/SkyIsland/WaterOnRocks/Spot_Water_Cave_River_Flowing_Cone_Int_Cue
root=QualityLevel_0
nodes [
 {
  id=Attenuation_0
  children [WavePlayer_3]
  properties {
   bOverrideAttenuation=true
   attenuationOverrides {
    attenuationShape=Cone
    bEnableFocusInterpolation=true
    bEnableListenerFocus=true
    falloffDistance=600
    focusAttackInterpSpeed=0.1
    focusReleaseInterpSpeed=0.1
    nonFocusVolumeAttenuation=0.75
    reverbWetLevelMin=0.6
    stereoSpread=100
    attenuationShapeExtents {y=28}
   }
  }
 }
 {
  id=Attenuation_1
  children [WavePlayer_1]
  properties {
   bOverrideAttenuation=true
   attenuationOverrides {attenuationShape=Cone, falloffDistance=750, reverbWetLevelMin=0.6, attenuationShapeExtents {y=25}}
  }
 }
 {
  id=Attenuation_2
  children [WavePlayer_2]
  properties {
   bOverrideAttenuation=true
   attenuationOverrides {attenuationShape=Cone, falloffDistance=1500, reverbWetLevelMin=0.6, attenuationShapeExtents {y=25}}
  }
 }
 {id=Mixer_0, children [Attenuation_2, Attenuation_1, Attenuation_0], properties {inputVolume [1,1,1]}}
 {id=QualityLevel_0, children [Mixer_0,Mixer_0,Mixer_0,Mixer_0]}
 {id=WavePlayer_1, properties {bLooping=true, soundWaveAssetPtr=/Game/Sounds/Fort_Ambient_Looping/Water/Spot_Water_Cave_River_Flowing}}
 {
  id=WavePlayer_2
  properties {
   bLooping=true
   soundWaveAssetPtr=/Game/Sounds/Artemis/POI/RealityFalls/Spots/Waterfalls/Ambience_RealityFalls_Waterfall_Medium_Bottom_Interior_Close_Loop
  }
 }
 {id=WavePlayer_3, properties {bLooping=true, soundWaveAssetPtr=/Game/Sounds/Apollo/Biomes/ATR/ATR_Waterfall_Loop_01}}
]
properties {
 maxDistance=1900
 soundClassObject=/Game/Sounds/FortSoundClassesAndModes/Fort_Ambient_Water
 volumeMultiplier=0.15
 attenuationOverrides {
  attenuationShape=Cone
  falloffDistance=750
  focusAzimuth=25
  focusVolumeAttenuation=1.2
  nonFocusAzimuth=45
  nonFocusVolumeAttenuation=0.5
  nonSpatializedRadiusStart=500
  stereoSpread=300
  attenuationShapeExtents=<638.1656 25>
 }
 concurrencyOverrides {maxCount=1, resolutionRule=PreventNew}
}
