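# Looping ambient water cue for the bottom of the Sky Island weir waterfall, capped by waterfall concurrency rules.
# The first two quality levels mix three waterfall and river loops, each with its own distance, low-pass and listener-focus attenuation.
# The last two quality levels play a single pitch-modulated river loop with wider falloff and occlusion.
# Nested mixers balance the layers, with one waterfall loop pitched up.
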SoundCue /Game/Sounds/Asteria/POI/SkyIsland/WaterOnRocks/Spot_Water_Main_Weir_Waterfall_Bottom_Cue
root=QualityLevel_0
nodes [
 {
  id=Attenuation_0
  children [Modulator_6]
  properties {
   bOverrideAttenuation=true
   attenuationOverrides {
    bAttenuateWithLPF=true
    bEnableFocusInterpolation=true
    bEnableListenerFocus=true
    distanceAlgorithm=Logarithmic
    falloffDistance=2000
    focusAttackInterpSpeed=0.2
    lPFFrequencyAtMax=1000
    lPFRadiusMax=4000
    lPFRadiusMin=1000
    nonFocusVolumeAttenuation=0.8
    nonSpatializedRadiusStart=2000
    occlusionInterpolationTime=2
    occlusionLowPassFilterFrequency=1000
    occlusionTraceChannel=ECC_GameTraceChannel11
    stereoSpread=500
    attenuationShapeExtents {x=1000}
   }
  }
 }
 {
  id=Attenuation_4
  children [WavePlayer_7]
  properties {
   bOverrideAttenuation=true
   attenuationOverrides {
    bAttenuateWithLPF=true
    bEnableFocusInterpolation=true
    bEnableListenerFocus=true
    distanceAlgorithm=NaturalSound
    falloffDistance=1350
    focusAttackInterpSpeed=0.1
    focusReleaseInterpSpeed=0.1
    lPFFrequencyAtMax=1000
    lPFRadiusMax=4000
    lPFRadiusMin=1000
    nonFocusVolumeAttenuation=0.6
    nonSpatializedRadiusStart=500
    occlusionInterpolationTime=3
    occlusionLowPassFilterFrequency=1000
    occlusionTraceChannel=ECC_GameTraceChannel11
    occlusionVolumeAttenuation=0.7
    reverbWetLevelMin=0.4
    stereoSpread=500
    attenuationShapeExtents {x=500}
   }
  }
 }
 {
  id=Attenuation_5
  children [WavePlayer_6]
  properties {
   bOverrideAttenuation=true
   attenuationOverrides {
    bAttenuateWithLPF=true
    bEnableFocusInterpolation=true
    bEnableListenerFocus=true
    falloffDistance=2000
    focusAttackInterpSpeed=0.1
    focusReleaseInterpSpeed=0.1
    lPFFrequencyAtMax=1000
    lPFRadiusMax=4000
    lPFRadiusMin=1000
    nonFocusVolumeAttenuation=0.2
    nonSpatializedRadiusStart=1000
    occlusionInterpolationTime=2
    occlusionLowPassFilterFrequency=1000
    occlusionTraceChannel=ECC_GameTraceChannel11
    stereoSpread=500
    attenuationShapeExtents {x=500}
   }
  }
 }
 {
  id=Attenuation_6
  children [Modulator_1]
  properties {
   bOverrideAttenuation=true
   attenuationOverrides {
    bAttenuateWithLPF=true
    bEnableFocusInterpolation=true
    bEnableListenerFocus=true
    bEnableOcclusion=true
    distanceAlgorithm=Logarithmic
    falloffDistance=4000
    lPFFrequencyAtMax=1000
    lPFRadiusMax=4000
    lPFRadiusMin=1000
    nonFocusVolumeAttenuation=0.8
    nonSpatializedRadiusStart=3000
    occlusionInterpolationTime=2
    occlusionLowPassFilterFrequency=1000
    occlusionTraceChannel=ECC_GameTraceChannel11
    stereoSpread=500
    attenuationShapeExtents {x=2000}
   }
  }
 }
 {id=Mixer_2, children [Attenuation_0, Mixer_4], properties {inputVolume [3,1]}}
 {id=Mixer_4, children [Attenuation_5, Attenuation_4], properties {inputVolume [3,0.25]}}
 {id=Modulator_1, children [WavePlayer_10], properties {pitchMax=1.1, pitchMin=0.9}}
 {id=Modulator_6, children [WavePlayer_1], properties {pitchMax=1.3, pitchMin=1.2}}
 {id=QualityLevel_0, children [Mixer_2, Mixer_2, Attenuation_6, Attenuation_6]}
 {id=WavePlayer_1, properties {bLooping=true, soundWaveAssetPtr=/Game/Athena/Sounds/Water/Ambient/Amb_Water_Waterfall_Large_Bot_Loop_01}}
 {
  id=WavePlayer_10
  properties {bLooping=true, soundWaveAssetPtr=/Game/Sounds/Artemis/POI/RealityFalls/Spots/Waterfalls/Ambience_RealityFalls_Waterfall_Large_Top_River_Loop}
 }
 {
  id=WavePlayer_6
  properties {bLooping=true, soundWaveAssetPtr=&Ambience_RealityFalls_Waterfall_Large_Top_River_Loop}
 }
 {
  id=WavePlayer_7
  properties {
   bLooping=true
   soundWaveAssetPtr=/Game/Sounds/Artemis/POI/RealityFalls/Spots/Waterfalls/Ambience_RealityFalls_Waterfall_Large_Bottom_Exterior_Close_Loop
  }
 }
]
properties {
 maxDistance=6000
 soundClassObject=/Game/Sounds/FortSoundClassesAndModes/Fort_Water_Ambient
 volumeMultiplier=1
 attenuationOverrides {
  bAttenuateWithLPF=true
  bEnableListenerFocus=true
  falloffDistance=6000
  lPFFrequencyAtMax=500
  nonSpatializedRadiusStart=700
  stereoSpread=1000
  attenuationShapeExtents {x=925}
 }
 concurrencyOverrides {maxCount=1}
 concurrencySet [/Game/Sounds/Concurrency_Rules/Waterfall_StopFarthestOldest_Conc]
}
