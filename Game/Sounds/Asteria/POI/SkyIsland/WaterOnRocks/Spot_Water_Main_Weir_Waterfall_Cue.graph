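# Looping ambient cue for the main weir waterfall at the Sky Island POI, built from reused waterfall loops and classed as water ambience.
# Close and wider exterior bottom loops use logarithmic, focus-aware attenuation with low-pass over distance.
# A pitched-up distant roar, whose custom curve fades in away from the source, mixes with a heavily boosted, pitched-up large waterfall loop.
# Both pairs feed one mixer shared by all quality levels.
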SoundCue /Game/Sounds/Asteria/POI/SkyIsland/WaterOnRocks/Spot_Water_Main_Weir_Waterfall_Cue
root=QualityLevel_0
nodes [
 {
  id=Attenuation_2
  children [WavePlayer_0]
  properties {
   bOverrideAttenuation=true
   attenuationOverrides {
    bAttenuateWithLPF=true
    bEnableFocusInterpolation=true
    bEnableListenerFocus=true
    distanceAlgorithm=Logarithmic
    falloffDistance=6000
    focusAttackInterpSpeed=0.1
    focusReleaseInterpSpeed=0.1
    lPFFrequencyAtMax=750
    lPFRadiusMax=7500
    lPFRadiusMin=1500
    nonFocusVolumeAttenuation=0.6
    nonSpatializedRadiusStart=2000
    occlusionInterpolationTime=1
    occlusionLowPassFilterFrequency=1000
    occlusionTraceChannel=ECC_GameTraceChannel11
    stereoSpread=500
    attenuationShapeExtents=<1500 1500>
   }
  }
 }
 {
  id=Attenuation_3
  children [WavePlayer_6]
  properties {
   bOverrideAttenuation=true
   attenuationOverrides {
    bAttenuateWithLPF=true
    bEnableFocusInterpolation=true
    bEnableListenerFocus=true
    distanceAlgorithm=Logarithmic
    falloffDistance=3000
    focusAttackInterpSpeed=0.1
    focusReleaseInterpSpeed=0.1
    focusVolumeAttenuation=1.2
    lPFFrequencyAtMax=1000
    lPFRadiusMax=5000
    lPFRadiusMin=1000
    nonFocusVolumeAttenuation=0.6
    nonSpatializedRadiusStart=500
    occlusionInterpolationTime=5
    occlusionLowPassFilterFrequency=1000
    occlusionTraceChannel=ECC_GameTraceChannel11
    occlusionVolumeAttenuation=0.5
    reverbWetLevelMin=0.4
    stereoSpread=500
    attenuationShapeExtents {x=1500}
   }
  }
 }
 {
  id=Attenuation_4
  children [Modulator_4]
  properties {
   bOverrideAttenuation=true
   attenuationOverrides {
    bEnableFocusInterpolation=true
    bEnableListenerFocus=true
    distanceAlgorithm=Custom
    falloffDistance=6000
    focusAttackInterpSpeed=0.1
    focusReleaseInterpSpeed=0.1
    lPFFrequencyAtMax=1000
    lPFRadiusMax=4000
    lPFRadiusMin=1000
    nonFocusVolumeAttenuation=0.6
    nonSpatializedRadiusStart=5000
    occlusionInterpolationTime=3
    occlusionLowPassFilterFrequency=1000
    occlusionTraceChannel=ECC_GameTraceChannel11
    stereoSpread=10000
    attenuationShapeExtents {x=2000}
    customAttenuationCurve {editorCurveData {keys [
       <at 0 0>
       {arriveTangent=0.028, interpMode=RCIM_Cubic, leaveTangent=0.028, tangentMode=RCTM_User, time=0.2209, value=1}
       {arriveTangent=-0.2554, interpMode=RCIM_Cubic, leaveTangent=-0.2554, tangentMode=RCTM_User, time=0.4, value=0.97}
       <at 1 0>
      ]}}
   }
  }
 }
 {
  id=Attenuation_5
  children [Modulator_2]
  properties {
   bOverrideAttenuation=true
   attenuationOverrides {
    bEnableFocusInterpolation=true
    distanceAlgorithm=Custom
    falloffDistance=6000
    lPFFrequencyAtMax=1000
    lPFRadiusMax=4000
    lPFRadiusMin=1000
    nonFocusVolumeAttenuation=0.8
    nonSpatializedRadiusEnd=2500
    nonSpatializedRadiusStart=5000
    occlusionInterpolationTime=3
    occlusionLowPassFilterFrequency=1000
    occlusionTraceChannel=ECC_GameTraceChannel11
    stereoSpread=10000
    attenuationShapeExtents {x=0}
    customAttenuationCurve {editorCurveData {keys [
       <at 0 1.2>
       {arriveTangent=0.028, interpMode=RCIM_Cubic, leaveTangent=0.028, tangentMode=RCTM_User, time=0.2209, value=1}
       {arriveTangent=-0.2554, interpMode=RCIM_Cubic, leaveTangent=-0.2554, tangentMode=RCTM_User, time=0.4, value=0.97}
       <at 1 0>
      ]}}
   }
  }
 }
 {id=Mixer_0, children [Attenuation_3, Attenuation_2], properties {inputVolume [1.5,1.5]}}
 {id=Mixer_1, children [Mixer_0,Mixer_2], properties {inputVolume [1,1]}}
 {id=Mixer_2, children [Attenuation_4, Attenuation_5], properties {inputVolume [1,5]}}
 {id=Modulator_2, children [WavePlayer_2], properties {pitchMax=1.2, pitchMin=1.1}}
 {id=Modulator_4, children [WavePlayer_1], properties {pitchMax=1.3, pitchMin=1.2}}
 {id=QualityLevel_0, children [Mixer_1,Mixer_1,Mixer_1,Mixer_1]}
 {
  id=WavePlayer_0
  properties {
   bLooping=true
   soundWaveAssetPtr=/Game/Sounds/Artemis/POI/RealityFalls/Spots/Waterfalls/Ambience_RealityFalls_Waterfall_Large_Bottom_Exterior_Loop
  }
 }
 {
  id=WavePlayer_1
  properties {bLooping=true, soundWaveAssetPtr=/Game/Sounds/Artemis/POI/RealityFalls/Spots/Waterfalls/Spot_RealityFalls_Waterfall_Roar_Medium_Distant_01}
 }
 {id=WavePlayer_2, properties {bLooping=true, soundWaveAssetPtr=/Game/Athena/Sounds/Water/Ambient/Amb_Water_Waterfall_Large_Bot_Loop_01}}
 {
  id=WavePlayer_6
  properties {
   bLooping=true
   soundWaveAssetPtr=/Game/Sounds/Artemis/POI/RealityFalls/Spots/Waterfalls/Ambience_RealityFalls_Waterfall_Large_Bottom_Exterior_Close_Loop
  }
 }
]
properties {
 maxDistance=8000
 soundClassObject=/Game/Sounds/FortSoundClassesAndModes/Fort_Water_Ambient
 volumeMultiplier=0.3
 attenuationOverrides {bSpatialize=false, falloffDistance=5000, nonSpatializedRadiusStart=500, stereoSpread=5000, attenuationShapeExtents {x=0}}
 concurrencyOverrides {maxCount=1, resolutionRule=PreventNew}
}
